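; Car prop Blueprint that varies hood, trunk and engine parts to dress environments.
; Its construction script, gated to editor or PIE, either randomizes hood and trunk rotation or applies set angles, with optional trunk inversion, then destroys parts according to visibility enums.
; Random removal derives from actor location plus per-part offsets, while random rotation also mixes in an unseeded random divisor.
; Helper functions close or randomize parts; event graph events are empty.
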(event EventBeginPlay)

(event Collision|EventActorBeginOverlap
  (params (OtherActor :type Actor%20Object%20Reference)))

(event EventTick
  (params (DeltaSeconds :type Float%20%28single-precision%29)))

(fn CloseAll
  (Utilities|FlowControl|Sequence
    (:then_0
      (PartsRandomization|CloseHood))
    (:then_1
      (PartsRandomization|CloseTrunk))))

(fn CloseHood
  (Transformation|ResetRelativeTransform (Variables|PartsRandomization|GetHood)))

(fn CloseTrunk
  (Transformation|ResetRelativeTransform (Variables|PartsRandomization|GetTrunk)))

(fn DestroyPart
  (params (Part :type Static%20Mesh%20Component%20Object%20Reference) (BeRandom :type Boolean) (Offset :type Float%20%28double-precision%29) (DestroyChance :type Float%20%28double-precision%29 :default 0.5))
  (if BeRandom
    (if (> (Class|EnvironmentArtFuncLib|RandomFloatbyLocation (* (Transformation|GetActorLocation) (Math|Conversions|ToVector(Float) Offset)) 15.0) DestroyChance)
      _))
  (Components|DestroyComponent Part))

(fn RandomizeAll
  (Utilities|FlowControl|Sequence
    (:then_0
      (PartsRandomization|RandomizeHood))
    (:then_1
      (PartsRandomization|RandomizeTrunk))))

(fn RandomizeHood
  (Transformation|SetRelativeRotation (Variables|PartsRandomization|GetHood) (Math|Rotator|Lerp(Rotator) "0, 0, 0" "0, 0,-60.000000" (Class|EnvironmentArtFuncLib|RandomFloatbyLocation (/ (Transformation|GetActorLocation) (Math|Random|RandomFloatinRange 1.0 1e+01)) 2e+01) true) false false))

(fn RandomizeTrunk
  (if (Variables|PartsRandomization|GetInvertTrunkAngle)
    (Transformation|SetRelativeRotation (Variables|PartsRandomization|GetTrunk) (Math|Rotator|InvertRotator (Math|Rotator|Lerp(Rotator) "0, 0, 0" "0, 0,105.000000" (Class|EnvironmentArtFuncLib|RandomFloatbyLocation (/ (Transformation|GetActorLocation) (Math|Random|RandomFloatinRange 1.0 1e+01)) 48.0) true)) false false)
    (else
      (Transformation|SetRelativeRotation (Variables|PartsRandomization|GetTrunk) (Math|Rotator|Lerp(Rotator) "0, 0, 0" "0, 0,60.000000" (Class|EnvironmentArtFuncLib|RandomFloatbyLocation (/ (Transformation|GetActorLocation) (Math|Random|RandomFloatinRange 1.0 1e+01)) 48.0) true) false false))))

(fn RemoveComponents
  (bind _hood (Variables|PartsRandomization|GetHood))
  (bind _engine (Variables|PartsRandomization|GetEngine))
  (bind _trunk (Variables|PartsRandomization|GetTrunk))
  (Utilities|FlowControl|Sequence
    (:then_0
      (switch Utilities|FlowControl|Switch|SwitchonCarPartVisibiltyEnum (Variables|PartsRandomization|GetTrunkVisibility)
        (:NewEnumerator0
          (PartsRandomization|DestroyPart _trunk true 0.7 0.76))
        (:NewEnumerator1
          (PartsRandomization|DestroyPart _trunk))))
    (:then_1
      (switch Utilities|FlowControl|Switch|SwitchonCarPartVisibiltyEnum (Variables|PartsRandomization|GetHoodVisibility)
        (:NewEnumerator0
          (PartsRandomization|DestroyPart _hood true 1.0 0.9))
        (:NewEnumerator1
          (PartsRandomization|DestroyPart _hood))))
    (:then_2
      (switch Utilities|FlowControl|Switch|SwitchonCarPartVisibiltyEnum (Variables|PartsRandomization|GetEngineVisibility)
        (:NewEnumerator0
          (PartsRandomization|DestroyPart _engine true 0.55))
        (:NewEnumerator1
          (PartsRandomization|DestroyPart _engine))))))

(fn SetHoodRotation
  (bind _hood (Variables|PartsRandomization|GetHood))
  (Transformation|ResetRelativeTransform _hood)
  (Transformation|AddLocalRotation _hood (Math|Rotator|InvertRotator (Math|Rotator|MakeRotator (Variables|PartsRandomization|GetHoodAngle))) false false))

(fn SetTrunkRotation
  (bind _trunk (Variables|PartsRandomization|GetTrunk))
  (bind _trunk_2 (Variables|PartsRandomization|GetTrunk))
  (if (Variables|PartsRandomization|GetInvertTrunkAngle)
    (Transformation|ResetRelativeTransform _trunk)
    (Transformation|AddLocalRotation _trunk (Math|Rotator|InvertRotator (Math|Rotator|MakeRotator (Variables|PartsRandomization|GetTrunkAngle))) false false)
    (else
      (Transformation|ResetRelativeTransform _trunk_2)
      (Transformation|AddLocalRotation _trunk_2 (Math|Rotator|MakeRotator (Variables|PartsRandomization|GetTrunkAngle)) false false))))

(fn ConstructionScript
  (|CallParentFunction)
  (if (Editor|InEditororPIE)
    (Utilities|FlowControl|Sequence
      (:then_0
        (if (Variables|PartsRandomization|GetUseRandomRotation)
          (Utilities|FlowControl|Sequence
            (:then_0
              (PartsRandomization|RandomizeHood))
            (:then_1
              (PartsRandomization|RandomizeTrunk)))
          (else
            (Utilities|FlowControl|Sequence
              (:then_0
                (PartsRandomization|SetHoodRotation))
              (:then_1
                (PartsRandomization|SetTrunkRotation))))))
      (:then_1
        (PartsRandomization|RemoveComponents)))))
